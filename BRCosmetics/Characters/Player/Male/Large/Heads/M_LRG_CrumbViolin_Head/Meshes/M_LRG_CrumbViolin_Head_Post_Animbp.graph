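# Post-process AnimBlueprint for the large male CrumbViolin character head, letting slot animation drive part of the tentacle beard during the character's custom emote.
# Each update checks whether the owning mesh component's anim instance is playing the Emote_CrumbViolin_CLM_M montage and stores 1 or 0.
# That value weights a layered per-bone blend of DefaultSlot over the input pose, filtered to the dyn_Tentacle_beard_04_r branch.
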AnimBlueprint /BRCosmetics/Characters/Player/Male/Large/Heads/M_LRG_CrumbViolin_Head/Meshes/M_LRG_CrumbViolin_Head_Post_Animbp
target_skeleton=/Game/Characters/Player/Common/Fortnite_Base_Head/Fortnite_Base_Head_Export_Skeleton
graphs [
 AnimGraph {nodes [Output Pose(Result=Layered blend per bone(BasePose=Input Pose(), BlendPoses_0=Slot 'DefaultSlot'(), BlendWeights_0=Get Is Playing Custom Emote() -> (Is Playing Custom Emote:double)) {layerSetup [{branchFilters [{boneName=dyn_Tentacle_beard_04_r}]}]})]}
 EventGraph {nodes [
   Event BlueprintUpdateAnimation()
   n1 = GetAnimInstance(self=GetOwnerSkeletalMeshComponent())
   > Set Is Playing Custom Emote(Is Playing Custom Emote:double=SelectFloat(A=1, B=0, bPickA=Montage_IsPlaying(self=@n1, Montage=/BRCosmetics/Animation/Game/MainPlayer/Emotes/CrumbViolin/CLM/Emote_CrumbViolin_CLM_M))) -> (Output_Get:double)
   TryGetPawnOwner()
  ]}
]
variables [Is Playing Custom Emote {type=number}]
